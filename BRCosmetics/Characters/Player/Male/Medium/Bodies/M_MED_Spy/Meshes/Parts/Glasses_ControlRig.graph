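# Control Rig for the male medium Spy character's glasses that nudges and tilts them as the cheeks rise.
# It reads the height of the left and right inner cheek bones relative to faceAttach.
# Each height is remapped with clamping, and the larger result weights a small additive local translation and roll on the dyn_glasses bone.
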ControlRigBlueprint /BRCosmetics/Characters/Player/Male/Medium/Bodies/M_MED_Spy/Meshes/Parts/Glasses_ControlRig
graphs [
 RigVMFunctionLibrary {links [], nodes []}
 RigVMModel {nodes [
   n1 = &RigUnit_BeginExecution()
   n2 = &RigUnit_GetRelativeBoneTransform(Bone=L_cheek_inner, Space=faceAttach)
   n3 = &RigUnit_GetRelativeBoneTransform(Bone=R_cheek_inner, Space=faceAttach)
   /Script/ControlRig.RigUnit_ModifyBoneTransforms(
    BoneToModify=((Bone="dyn_glasses",Transform=(Rotation=(X=0,Y=0,Z=0,W=1),Translation=(X=0,Y=0.5,Z=0.5),Scale3D=(X=1,Y=1,Z=1))))
    Mode=AdditiveLocal
    WeightMaximum=1
    WeightMinimum=0
    ExecutePin=@n1.ExecutePin
    Weight=/Script/RigVM.RigVMFunction_MathFloatMax(
     A=&RigVMFunction_MathFloatRemap(SourceMaximum=5, SourceMinimum=3.86, TargetMinimum=0, bClamp=true, Value=@n2.Transform.Translation.Z).Result
     B=&RigVMFunction_MathFloatRemap(SourceMaximum=5, SourceMinimum=3.86, TargetMinimum=0, bClamp=true, Value=@n3.Transform.Translation.Z).Result
    ) {is=struct RigVMUnitNode, template="Maximum::Execute(in A,in B,out Result)"}.Result
    BoneToModify.0.Transform.Rotation=/Script/RigVM.RigVMFunction_MathQuaternionFromRotator(Rotator=(Pitch=0,Yaw=0,Roll=-3)) {is=struct RigVMUnitNode}.Result
   ) {is=struct RigVMUnitNode}
  ]}
]
hierarchy [
 root {loc=<0 0 0>, rot=<0 0 0 1>}
 pelvis {parent=root, loc=<0 -0.7071 93.1191>, rot=<0.0127 0.707 -0.0127 -0.707>}
 spine_01 {parent=pelvis, loc=<6.7494 -0.6388 0>, rot=<0 0 0.034 -0.9994>}
 spine_02 {parent=spine_01, loc=<7.2518 -0.035 0>, rot=<0 0 0.0238 -0.9997>}
 spine_03 {parent=spine_02, loc=<6.0104 0 0>, rot=<0 0 0.0149 -0.9999>}
 spine_04 {parent=spine_03, loc=<8.1024 0 0>, rot=<0 0 -0.0132 -0.9999>}
 spine_05 {parent=spine_04, loc=<10.0146 0 0>, rot=<0 0 -0.0015 -1>}
 neck_01 {parent=spine_05, loc=<11.3138 0.1529 0>, rot=<0 0 -0.2003 -0.9797>}
 neck_02 {parent=neck_01, loc=<4.5627 0 0>, rot=<0 0 0.0504 -0.9987>}
 head {parent=neck_02, loc=<5.3914 0 0>, rot=<0 0 0.115 -0.9934>}
 faceAttach {parent=head, loc=<-2.6771 0.77 -0.0072>, rot=<0.0082 -0.7071 -0.0082 -0.7071>, scale=<0.89 0.89 0.89>}
 C_brow_mid {parent=faceAttach, loc=<-0.0289 10.034 8.5003>, rot=<0 0.0185 -0.0037 -0.9998>}
 L_brow_mid {parent=faceAttach, loc=<3.3976 9.225 9.44>, rot=<-0.0001 -0.0043 0.2553 -0.9669>}
 L_brow_outer {parent=faceAttach, loc=<5.467 7.3929 9.5954>, rot=<0.1272 0.1563 0.4034 -0.8926>}
 L_cheek_inner {parent=faceAttach, loc=<4.2162 7.5794 3.7472>, rot=<-0.0363 -0.1829 0.329 -0.9257>}
 C_lip_upper_mid {parent=faceAttach, loc=<-0.0062 9.7996 1.2163>, rot=<0.2225 0.0019 -0.0007 -0.9749>}
 L_lip_upper_outer {parent=faceAttach, loc=<1.803 8.6976 1.2476>, rot=<0.0414 -0.0575 0.2569 -0.9638>}
 L_lip_corner {parent=faceAttach, loc=<2.7324 8.1835 1.1548>, rot=<0.0018 0.0092 -0.0145 -0.9999>}
 C_nose_bridge {parent=faceAttach, loc=<0.0087 11.2686 5.6021>, rot=<-0.2218 -0.0214 -0.006 -0.9748>}
 R_brow_mid {parent=faceAttach, loc=<-3.4123 9.2266 9.4451>, rot=<-0.0001 0.0043 -0.2553 -0.9669>}
 R_brow_outer {parent=faceAttach, loc=<-5.4923 7.3805 9.5792>, rot=<0.066 -0.0792 -0.4248 -0.8994>}
 R_cheek_inner {parent=faceAttach, loc=<-4.1711 7.5906 3.62>, rot=<-0.0372 0.185 -0.3288 -0.9253>}
 R_lip_corner {parent=faceAttach, loc=<-2.7045 8.1398 1.1742>, rot=<0.0027 -0.0083 0.0146 -0.9999>}
 C_jaw {parent=faceAttach, loc=<-0.0398 3.0022 1.3275>, rot=<0 0 0 1>}
 L_lip_lower_outer {parent=C_jaw, loc=<1.8753 5.5956 -0.1435>, rot=<0.0239 0.1036 0.3495 -0.9309>}
 C_lip_lower_mid {parent=C_jaw, loc=<0.0215 6.665 -0.1766>, rot=<0 0.0164 0.0035 -0.9999>}
 R_lip_lower_outer {parent=C_jaw, loc=<-1.821 5.7427 -0.2291>, rot=<0.0233 -0.1007 -0.3474 -0.932>}
 R_eye {parent=faceAttach, loc=<-3.2511 6.962 8.0669>, rot=<0 0 0.9998 -0.0187>}
 R_eye_lid_upper_mid {parent=R_eye, loc=<-0.0396 -1.1714 0.2631>, rot=<0.0523 -0.0009 0.0167 -0.9985>}
 R_eye_lid_lower_mid {parent=R_eye, loc=<-0.0085 -0.9425 -0.7051>, rot=<0.0329 -0.0005 0.0167 -0.9993>}
 L_eye {parent=faceAttach, loc=<3.235 6.962 8.0669>, rot=<0 0 0.9998 0.0187>}
 L_eye_lid_upper_mid {parent=L_eye, loc=<0.0251 -1.1623 0.2388>, rot=<0.0523 0 0 -0.9986>}
 L_eye_lid_lower_mid {parent=L_eye, loc=<0.0223 -0.923 -0.7488>, rot=<0.0329 0 0 -0.9995>}
 R_lip_upper_outer {parent=faceAttach, loc=<-1.8403 8.7479 1.2822>, rot=<0.043 0.0528 -0.2551 -0.9645>}
 dyn_hair_c_a {parent=head, loc=<14.5267 4.885 -2.9543>, rot=<0.0029 -0.7071 -0.0029 -0.7071>}
 dyn_hair_front_b {parent=head, loc=<13.8082 8.9403 -0.3061>, rot=<0.0029 -0.7071 -0.0029 -0.7071>}
 dyn_hair_front_a {parent=head, loc=<12.9164 9.4167 -1.1412>, rot=<0.0029 -0.7071 -0.0029 -0.7071>}
 dyn_hair_b_e {parent=head, loc=<15.2735 2.4786 2.7445>, rot=<0.0029 -0.7071 -0.0029 -0.7071>}
 dyn_hair_c_d {parent=head, loc=<14.0913 -2.3831 0.1059>, rot=<0.0029 -0.7071 -0.0029 -0.7071>}
 dyn_hair_c_b {parent=head, loc=<15.2094 2.4601 -1.7529>, rot=<0.0029 -0.7071 -0.0029 -0.7071>}
 dyn_hair_d_f {parent=head, loc=<11.8722 -4.0203 -1.4239>, rot=<0.0029 -0.7071 -0.0029 -0.7071>}
 dyn_hair_d_e {parent=head, loc=<12.1245 -2.5869 -3.5806>, rot=<0.0029 -0.7071 -0.0029 -0.7071>}
 dyn_hair_d_d {parent=head, loc=<12.8085 -1.1761 -3.4967>, rot=<0.0029 -0.7071 -0.0029 -0.7071>}
 dyn_hair_d_c {parent=head, loc=<13.6243 0.9535 -3.4437>, rot=<0.0029 -0.7071 -0.0029 -0.7071>}
 dyn_hair_d_b {parent=head, loc=<13.2684 3.4202 -3.9695>, rot=<0.0029 -0.7071 -0.0029 -0.7071>}
 dyn_hair_d_a {parent=head, loc=<12.7684 6.3653 -3.4566>, rot=<0.0029 -0.7071 -0.0029 -0.7071>}
 dyn_hair_a_b {parent=head, loc=<14.1299 4.8399 2.7103>, rot=<0.0029 -0.7071 -0.0029 -0.7071>}
 dyn_hair_b_d {parent=dyn_hair_a_b, loc=<1.4323 -0.4377 1.4411>, rot=<0 0 0 1>}
 dyn_hair_b_c {parent=dyn_hair_a_b, loc=<1.0587 0.8539 1.129>, rot=<0 0 0 1>}
 dyn_hair_a_a {parent=head, loc=<14.1147 7.2262 1.2815>, rot=<0.0992 -0.7001 0.0934 -0.7009>}
 dyn_hair_b_b {parent=dyn_hair_a_a, loc=<1.2784 0.2508 1.211>, rot=<0 0 -0.1362 -0.9907>}
 dyn_hair_b_a {parent=dyn_hair_a_a, loc=<2.8695 0.5996 0.566>, rot=<0 0 -0.0644 -0.9979>}
 dyn_hair_a_c {parent=head, loc=<12.93 -0.698 3.933>, rot=<0.0029 -0.7071 -0.0029 -0.7071>}
 dyn_hair_b_g {parent=dyn_hair_a_c, loc=<0.5289 -1.2776 0.6308>, rot=<0 0 0 1>}
 dyn_hair_b_f {parent=dyn_hair_a_c, loc=<0.4204 0.3869 1.6033>, rot=<0 0 0 1>}
 dyn_ear_base {parent=head, loc=<0.5395 1.6188 6.5994>, rot=<0.0041 -1 0 0>}
 dyn_ear {parent=dyn_ear_base, loc=<0 0 0>, rot=<0 0 0 1>}
 dyn_glasses_base {parent=head, loc=<4.8581 -0.2062 0>, rot=<0.0029 -0.7071 -0.0029 -0.7071>}
 dyn_glasses {parent=dyn_glasses_base, loc=<0 0 0>, rot=<0 0 0 1>}
 dyn_ear_bottom {parent=neck_02, loc=<1.4329 1.8968 4.9611>, rot=<-0.0784 -0.7027 0.0784 -0.7027>}
]
