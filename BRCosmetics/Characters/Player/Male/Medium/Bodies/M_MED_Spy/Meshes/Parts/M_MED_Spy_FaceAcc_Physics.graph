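# Physics asset for the male medium Spy body's face accessory part, covering its dynamic hair and ear bones with small capsule bodies.
# Most hair bones constrain directly to the head with tight swing and twist limits, while most of the b-group strands chain off their a-group parents.
# The ear is a two-link chain: its base rotates freely on all angular axes against the head, and the outer link is limited relative to that base.
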PhysicsAsset /BRCosmetics/Characters/Player/Male/Medium/Bodies/M_MED_Spy/Meshes/Parts/M_MED_Spy_FaceAcc_Physics
bodies {
 dyn_ear [{shapeType=Capsule, length=1.1, radius=1.1, center=<2.3235 0 0>, rotation=<rot 90 0 0>}]
 dyn_ear_base [{shapeType=Capsule, length=1, radius=1}]
 dyn_hair_a_a [{shapeType=Capsule, length=1.1, radius=1.2, center=<0 0 -1>}]
 dyn_hair_a_b [{shapeType=Capsule, length=1.1, radius=1.2, center=<0 0 -1>}]
 dyn_hair_a_c [{shapeType=Capsule, length=1.1, radius=1.2, center=<0 0 -1>}]
 dyn_hair_b_a [{shapeType=Capsule, length=0.8, radius=0.5, center=<0 0 -0.5>}]
 dyn_hair_b_b [{shapeType=Capsule, length=0.8, radius=0.5, center=<0 0 -0.5>}]
 dyn_hair_b_c [{shapeType=Capsule, length=0.8, radius=0.5, center=<0 0 -0.5>}]
 dyn_hair_b_d [{shapeType=Capsule, length=0.8, radius=0.5, center=<0 0 -0.5>}]
 dyn_hair_b_e [{shapeType=Capsule, length=0.9, radius=0.7, center=<0 0 -1>}]
 dyn_hair_b_f [{shapeType=Capsule, length=0.8, radius=0.5, center=<0 0 -0.5>}]
 dyn_hair_b_g [{shapeType=Capsule, length=0.8, radius=0.5, center=<0 0 -0.5>}]
 dyn_hair_c_a [{shapeType=Capsule, length=0.9, radius=0.9, center=<0 0 -1>}]
 dyn_hair_c_b [{shapeType=Capsule, length=0.9, radius=0.9, center=<0 0 -1>}]
 dyn_hair_c_d [{shapeType=Capsule, length=0.9, radius=0.9, center=<0 0 -1>}]
 dyn_hair_d_a [{shapeType=Capsule, length=1.1, radius=0.9, center=<0 0 -1>}]
 dyn_hair_d_b [{shapeType=Capsule, length=1.1, radius=0.9, center=<0 0 -1>}]
 dyn_hair_d_c [{shapeType=Capsule, length=1.1, radius=0.9, center=<0 0 -1>}]
 dyn_hair_d_d [{shapeType=Capsule, length=1.1, radius=0.9, center=<0 0 -1>}]
 dyn_hair_d_e [{shapeType=Capsule, length=1.1, radius=0.9, center=<0 0 -1>}]
 dyn_hair_d_f [{shapeType=Capsule, length=1.1, radius=0.9, center=<0 0 -1>}]
 dyn_hair_front_a [{shapeType=Capsule, length=0.9, radius=0.9, center=<0 0 -1>}]
 dyn_hair_front_b [{shapeType=Capsule, length=0.9, radius=0.9, center=<0 0 -1>}]
 head [{shapeType=Capsule, length=1, radius=1}]
}
constraints [
 {bone1Name=dyn_hair_a_b, bone2Name=head, swing1LimitDegrees=10, swing2LimitDegrees=10, twistLimitDegrees=5}
 {bone1Name=dyn_hair_a_a, bone2Name=head, swing1LimitDegrees=10, swing2LimitDegrees=10, twistLimitDegrees=5}
 {bone1Name=dyn_hair_a_c, bone2Name=head, swing1LimitDegrees=10, swing2LimitDegrees=10, twistLimitDegrees=5}
 {bone1Name=dyn_hair_b_d, bone2Name=dyn_hair_a_b, swing1LimitDegrees=10, swing2LimitDegrees=10, twistLimitDegrees=5}
 {bone1Name=dyn_hair_b_c, bone2Name=dyn_hair_a_b, swing1LimitDegrees=10, swing2LimitDegrees=10, twistLimitDegrees=5}
 {bone1Name=dyn_hair_b_b, bone2Name=dyn_hair_a_a, swing1LimitDegrees=10, swing2LimitDegrees=10, twistLimitDegrees=5}
 {bone1Name=dyn_hair_b_a, bone2Name=dyn_hair_a_a, swing1LimitDegrees=10, swing2LimitDegrees=10, twistLimitDegrees=5}
 {bone1Name=dyn_hair_b_g, bone2Name=dyn_hair_a_c, swing1LimitDegrees=10, swing2LimitDegrees=10, twistLimitDegrees=5}
 {bone1Name=dyn_hair_b_f, bone2Name=dyn_hair_a_c, swing1LimitDegrees=10, swing2LimitDegrees=10, twistLimitDegrees=5}
 {bone1Name=dyn_hair_c_d, bone2Name=head, swing1LimitDegrees=15, swing2LimitDegrees=15, twistLimitDegrees=5}
 {bone1Name=dyn_hair_c_b, bone2Name=head, swing1LimitDegrees=15, swing2LimitDegrees=15, twistLimitDegrees=5}
 {bone1Name=dyn_hair_c_a, bone2Name=head, swing1LimitDegrees=15, swing2LimitDegrees=15, twistLimitDegrees=5}
 {bone1Name=dyn_hair_front_b, bone2Name=head, swing1LimitDegrees=15, swing2LimitDegrees=15, twistLimitDegrees=5}
 {bone1Name=dyn_hair_front_a, bone2Name=head, swing1LimitDegrees=15, swing2LimitDegrees=15, twistLimitDegrees=5}
 {bone1Name=dyn_hair_d_f, bone2Name=head, swing1LimitDegrees=20, swing2LimitDegrees=20, twistLimitDegrees=5}
 {bone1Name=dyn_hair_d_e, bone2Name=head, swing1LimitDegrees=20, swing2LimitDegrees=20, twistLimitDegrees=5}
 {bone1Name=dyn_hair_d_d, bone2Name=head, swing1LimitDegrees=20, swing2LimitDegrees=20, twistLimitDegrees=5}
 {bone1Name=dyn_hair_d_c, bone2Name=head, swing1LimitDegrees=20, swing2LimitDegrees=20, twistLimitDegrees=5}
 {bone1Name=dyn_hair_d_b, bone2Name=head, swing1LimitDegrees=20, swing2LimitDegrees=20, twistLimitDegrees=5}
 {bone1Name=dyn_hair_d_a, bone2Name=head, swing1LimitDegrees=20, swing2LimitDegrees=20, twistLimitDegrees=5}
 {bone1Name=dyn_hair_b_e, bone2Name=head, swing1LimitDegrees=15, swing2LimitDegrees=15, twistLimitDegrees=5}
 {
  bone1Name=dyn_ear_base
  bone2Name=head
  swing1LimitDegrees=45
  swing1Motion=Free
  swing2LimitDegrees=45
  swing2Motion=Free
  twistLimitDegrees=45
  twistMotion=Free
 }
 {bone1Name=dyn_ear, bone2Name=dyn_ear_base, swing1LimitDegrees=20, swing2LimitDegrees=20, twistLimitDegrees=5}
]
