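# Post-copy AnimBlueprint for the male medium Spy's face accessory mesh, driving ear and glasses secondary motion.
# It copies the attached parent's pose, constrains ear bones to head and neck, then runs RigidBody physics in spine_05 space.
# A glasses Control Rig finishes the pose, enabled only at LOD0 by the event graph.
# The montage-weighted additive glasses-bone nudge is not wired into the output.
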AnimBlueprint /BRCosmetics/Characters/Player/Male/Medium/Bodies/M_MED_Spy/Meshes/Parts/M_MED_Spy_FaceAcc_AnimBP
target_skeleton=&/M_MED_Spy_FaceAcc_Skeleton
graphs [
 AnimGraph {nodes [
   "Transform (Modify) Bone - Bone: dyn_glasses_base"(Alpha=Get montageIsActive() -> (montageIsActive:double), Translation="0,0.4,0.4", Rotation="0,0,-3", Scale="1,1,1") {
    rotationMode=BMM_Additive
    rotationSpace=BCS_BoneSpace
    translationMode=BMM_Additive
    translationSpace=BCS_BoneSpace
    alphaBoolBlend {blendOption=Linear}
    alphaScaleBiasClamp {bInterpResult=true}
    boneToModify {boneName=dyn_glasses_base}
    rotation {roll=-3}
    translation {y=0.4, z=0.4}
   }
   n1 = "Constraint - dyn_ear_base (PP)"!(
    ComponentPose=Local To Component(LocalPose=Copy Pose From Mesh() {bUseAttachedParent=true})
    Alpha=1
    ConstraintWeights_0=0.5
    ConstraintWeights_1:float=0.5
   ) -> (Pose) {
    alphaBoolBlend {blendOption=Linear}
    boneToModify {boneName=dyn_ear_base}
    constraintSetup [{targetBone {boneName=head}}, {targetBone {boneName=neck_02}}]
    constraintWeights [1,1]
   }
   n2 = "Constraint - dyn_ear_bottom (PPP)"!(ComponentPose=@n1, Alpha=1, ConstraintWeights_0=0.4124, ConstraintWeights_1:float=0.1869, ConstraintWeights_2:float=0.4007) -> (Pose) {
    alphaBoolBlend {blendOption=Linear}
    boneToModify {boneName=dyn_ear_bottom}
    constraintSetup [{targetBone {boneName=neck_02}}, {targetBone {boneName=neck_01}}, {targetBone {boneName=head}}]
    constraintWeights [1,1,1]
   }
   n3 = RigidBody(ComponentPose=@n2) {
    bClampLinearTranslationLimitToRefPose=true
    bOverrideWorldGravity=true
    overridePhysicsAsset=&/M_MED_Spy_FaceAcc_Physics
    simulationSpace=BaseBoneSpace
    alphaBoolBlend {blendOption=Linear}
    baseBoneRef {boneName=spine_05}
    componentLinearAccScale=<0.4 0.4 0.4>
    componentLinearVelScale=<0.4 0.4 0.4>
    overrideWorldGravity {z=-980}
   }
   n4 = Control Rig(Source=Component To Local(ComponentPose=@n3), Alpha=Get isLOD0() -> (isLOD0:double)) {
    alphaBoolBlend {blendOption=Linear}
    controlRigAssetReference {blueprintRigClass=&/Glasses_ControlRig.Glasses_ControlRig_C}
   }
   Output Pose(Result=@n4) {layerGroup=None}
  ]}
 EventGraph {nodes [
   Event BlueprintUpdateAnimation()
   > n1 = Sequence()
   n2 = GetAnimInstance(self=GetOwnerSkeletalMeshComponent())
   > Is Valid(InputObject=@n2)
   > Set montageIsActive(montageIsActive:double="To Float (Boolean)"(InBool=Montage_IsActive!(self:Anim Instance Object Reference=@n2, Montage:Anim Montage Object Reference) -> (ReturnValue:bool))) -> (Output_Get:double)
   n3 = Branch(execute=@n1.then_1, Condition="Equal (Integer)"(A=Get CurrentLOD(), B=0) {is=title K2Node_CallFunction})
   > Set isLOD0(isLOD0:double=1) -> (Output_Get:double)
   Set isLOD0(execute=@n3.else, isLOD0=0)
   TryGetPawnOwner()
  ]}
]
variables [
 isLOD0 {type=number}
 montageIsActive {type=number}
]
